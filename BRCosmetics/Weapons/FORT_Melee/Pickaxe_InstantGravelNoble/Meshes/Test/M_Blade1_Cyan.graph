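# Test-folder cyan blade material for the InstantGravelNoble pickaxe: translucent, unlit, usable on skeletal and Niagara meshes.
# Glow combines a curve-atlas gradient along UV1, fresnel, sky reflection, a center streak and two noise-distorted, panning plasma layers.
# A noise-offset DissolveProgress threshold adds a bright edge, while UV1 gradients build a separate tip-to-bottom edge-fade mask.
# TwoSidedSign swaps back faces to a flat color, and particle color tints the result.
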Material /BRCosmetics/Weapons/FORT_Melee/Pickaxe_InstantGravelNoble/Meshes/Test/M_Blade1_Cyan
graph {
 properties {
  bUsedWithMeshDeformer=true
  bUsedWithNiagaraMeshParticles=true
  bUsedWithSkeletalMesh=true
  blendMode=BLEND_Translucent
  opacityMaskClipValue=0
  pixelDepthOffsetMode=PDOM_Legacy
  shadingModel=MSM_Unlit
  usedShadingModels=Unlit
 }
 nodes [
  Constant() {r=1}
  Constant()
  n1 = TextureSample(UVs=TextureCoordinate()) {samplerType=SAMPLERTYPE_Normal, texture=/BRCosmetics/Weapons/FORT_Melee/Pickaxe_InstantGravelNoble/Textures/T_Pickaxe_InstantGravelNoble_N}
  n2 = Saturate(None=TwoSidedSign())
  LinearInterpolate(
   A=Transform(None=Constant3Vector() {constant=<rgba 0 0 1 0>}) {transformSourceType=TRANSFORMSOURCE_Local, transformType=TRANSFORM_Tangent}
   B=@n1.RGB
   Alpha=@n2
  )
  n3 = TextureCoordinate() {coordinateIndex=1}
  n4 = TextureObjectParameter() {
   group=Dissolve
   parameterName=Dissolve_Texture
   samplerType=SAMPLERTYPE_LinearGrayscale
   sortPriority=0
   texture=/Game/Effects/Fort_Effects/Emotes/UltraLight/Texture/T_Tiling_Noise_01_GreyScale
  }
  n5 = VectorParameter() {group=Dissolve, parameterName=Dissolve_UVControl, channelNames {a=VPan, b=UPan, g=VScale, r=UScale}, defaultValue=<rgba 1 0.2 0 0.1>}
  n6 = VectorParameter() {
   group=Dissolve
   parameterName=Dissolve_NoiseControl
   sortPriority=60
   channelNames {a=VStrength, b=UStrength, g=VOffset, r=UOffset}
   defaultValue=<rgba -0.208 -0.5 0.1 0>
  }
  n7 = MaterialFunctionCall!(
   NoiseValue=MaterialFunctionCall(Texture=@n4, UVControl_RGB=@n5, UVControl_A=@n5, "[Override]UVInput"=@n3) {materialFunction=&FXCommon_TextureSample}.RG
   Control_RGB=@n6
   Control_A=@n6
  ) {materialFunction=/Game/Effects/Niagara/MasterMaterials/Functions/Common/FXCommon_NoiseControl}
  n8 = Subtract(
   A=MaterialFunctionCall(Float2=@n3) {materialFunction=&BreakOutFloat2Components}.G
   B=Add(A=ScalarParameter() {defaultValue=-0.1, parameterName=DissolveProgress}, B=ComponentMask(None=@n7.NoiseValue) {r=true})
  )
  Multiply(A=@n8)
  n9 = MaterialFunctionCall(Float2=TextureCoordinate() {coordinateIndex=1}) {materialFunction=&BreakOutFloat2Components}
  n10 = Multiply(A=Min(A=OneMinus(None=@n9.R), B=@n9.R)) {constB=2}
  n11 = Add(A=@n10, B=ScalarParameter() {defaultValue=-0.7, group=TipEdge, parameterName=TipEdge_GrayAdjust_Add})
  n12 = ScalarParameter() {defaultValue=4.5, group=TipEdge, parameterName=TipEdge_GrayAdjust_Mul}
  n13 = Power(Base=Saturate(None=Multiply(A=@n11, B=@n12)), Exp=ScalarParameter() {defaultValue=5, group=TipEdge, parameterName=TipEdge_GrayAdjust_Power})
  n14 = Add(A=@n10, B=ScalarParameter() {defaultValue=0.2, group=BottomEdge, parameterName=BottomEdge_GrayAdjust_Add})
  n15 = ScalarParameter() {defaultValue=1.5, group=BottomEdge, parameterName=BottomEdge_GrayAdjust_Mul}
  n16 = Power(
   Base=Saturate(None=Multiply(A=@n14, B=@n15))
   Exp=ScalarParameter() {defaultValue=2, group=BottomEdge, parameterName=BottomEdge_GrayAdjust_Power}
  )
  n17 = Add(A=@n9.G, B=ScalarParameter() {defaultValue=-0.5, group=EdgeGrad, parameterName=EdgeGrad_GrayAdjust_Add})
  n18 = ScalarParameter() {defaultValue=2, group=EdgeGrad, parameterName=EdgeGrad_GrayAdjust_Mul}
  n19 = Power(
   Base=Saturate(None=Multiply(A=@n17, B=@n18))
   Exp=ScalarParameter() {defaultValue=2, group=EdgeGrad, parameterName=EdgeGrad_GrayAdjust_Power}
  )
  Multiply(
   A=LinearInterpolate(Alpha=OneMinus(None=@n9.G)) {constA=1, constB=0.2}
   B=LinearInterpolate(A=Saturate(None=@n13), B=Saturate(None=@n16), Alpha=Saturate(None=@n19))
  )
  n20 = ScalarParameter() {defaultValue=0.1, parameterName=DissolveOffset}
  n21 = Saturate(None=OneMinus(None=Divide(A=@n8, B=@n20) {constB=0.2}))
  n22 = VectorParameter() {parameterName=MaskRefine, defaultValue=<rgba 30 10 2 1>}
  n23 = LinearInterpolate(
   A=Multiply(A=Power(Base=@n21, Exp=@n22.R), B=@n22.G)
   B=Multiply(A=Power(Base=@n21, Exp=@n22.B), B=@n22.A)
   Alpha=ScalarParameter() {defaultValue=0.5, parameterName=ACACACDW}
  )
  n24 = VectorParameter() {parameterName=CCCC, defaultValue=<rgba 0 2.6191 10 0.5>}
  n25 = CurveAtlasRowParameter(CurveTime=MaterialFunctionCall(Float2=TextureCoordinate() {coordinateIndex=1}) {materialFunction=&BreakOutFloat2Components}.G) {
   atlas=/Game/Weapons/FORT_Melee/Pickaxe_InstantGravelNoble/Meshes/Test/CLA_InstantGraveNoble
   curve=/Game/Weapons/FORT_Melee/Pickaxe_InstantGravelNoble/Meshes/Test/CL_BladeColor_Cyan
   defaultValue=4
   parameterName=Color
  }
  n26 = VectorParameter() {parameterName=FresnelColor, defaultValue=<rgba 0.3106 0 2 1>}
  n27 = Fresnel(ExponentIn=ScalarParameter() {defaultValue=4, parameterName=Exp}, BaseReflectFractionIn=ScalarParameter() {parameterName=BR})
  n28 = Divide(A=Add(A=@n25, B=Multiply(A=@n26.RGB, B=@n27)))
  n29 = ScalarParameter() {defaultValue=0.05, parameterName=FFF}
  n30 = SkyLightEnvMapSample(Direction=ReflectionVectorWS(CustomWorldNormal=PixelNormalWS()), Roughness=Constant())
  n31 = MaterialFunctionCall(Float2=TextureCoordinate() {coordinateIndex=1}) {materialFunction=&BreakOutFloat2Components}
  n32 = Multiply(A=OneMinus(None=@n31.G), B=ScalarParameter() {defaultValue=1.5, parameterName=ACAC})
  n33 = Power(Base=Sine(None=Multiply(A=@n31.R) {constB=0.5}), Exp=ScalarParameter() {defaultValue=200, parameterName=CenterPO})
  n34 = VectorParameter() {parameterName=CenterColor, defaultValue=<rgba 0 0.2404 1 1>}
  n35 = Add(
   A=Add(
    A=MaterialFunctionCall(Emissive=@n28) {materialFunction=&MF_FX_EmissiveAdjustments}.Result
    B=Multiply(A=@n29, B=@n30)
   )
   B=Multiply(A=Saturate(None=@n32), B=Multiply(A=@n33, B=@n34.RGB))
  )
  n36 = OneMinus(None=@n31.G)
  n37 = ScalarParameter() {defaultValue=10, parameterName=ACAC333}
  n38 = VectorParameter() {
   group=TextureSampler
   parameterName=TextureSampler_G2C_Refine
   channelNames {a=Blend, b=B_Mul, g=A_Mul, r=A_Power}
   defaultValue=<rgba 2 1 0.5 0.5>
  }
  n39 = TextureObjectParameter() {
   group=TextureSampler
   parameterName=TextureSampler_Texture
   samplerType=SAMPLERTYPE_LinearGrayscale
   sortPriority=0
   texture=/Game/Effects/Fort_Effects/Emotes/UltraLight/Texture/T_Plasma_C_Greyscale
  }
  n40 = TextureObjectParameter() {
   group=NoiseA
   parameterName=NoiseA_Texture
   samplerType=SAMPLERTYPE_LinearGrayscale
   sortPriority=0
   texture=&T_SimplePerlin_Greyscale
  }
  n41 = VectorParameter() {group=NoiseA, parameterName=NoiseA_UVControl, channelNames {a=VPan, b=UPan, g=VScale, r=UScale}, defaultValue=<rgba 2 1 0 0.1>}
  n42 = VectorParameter() {
   group=NoiseA
   parameterName=NoiseA_NoiseControl
   sortPriority=60
   channelNames {a=VStrength, b=UStrength, g=VOffset, r=UOffset}
   defaultValue=<rgba -0.5 -0.5 0.125 0>
  }
  n43 = MaterialFunctionCall(
   NoiseValue=MaterialFunctionCall(Texture=@n40, UVControl_RGB=@n41, UVControl_A=@n41) {materialFunction=&FXCommon_TextureSample}.RG
   Control_RGB=@n42
   Control_A=@n42
  ) {materialFunction=&FXCommon_NoiseControl}
  n44 = VectorParameter() {
   group=TextureSampler
   parameterName=TextureSampler_UVControl
   channelNames {a=VPan, b=UPan, g=VScale, r=UScale}
   defaultValue=<rgba 4 0.5 0 0.25>
  }
  n45 = MaterialFunctionCall(Texture=@n39, AdditionalUVOffset=@n43.NoiseValue, UVControl_RGB=@n44, UVControl_A=@n44) {materialFunction=&FXCommon_TextureSample}
  n46 = TextureObjectParameter() {
   group=Textureflow2
   parameterName=Textureflow2_Texture
   samplerType=SAMPLERTYPE_LinearGrayscale
   sortPriority=0
   texture=/Game/Effects/Fort_Effects/Emotes/UltraLight/Texture/T_Plasma_B_Greyscale
  }
  n47 = VectorParameter() {group=Textureflow2, parameterName=Textureflow2_UVControl, channelNames {a=VPan, b=UPan, g=VScale, r=UScale}, defaultValue=<rgba 2 0.5 0 0.2>}
  n48 = MaterialFunctionCall(Texture=@n46, AdditionalUVOffset=@n43.NoiseValue, UVControl_RGB=@n47, UVControl_A=@n47) {materialFunction=&FXCommon_TextureSample}
  n49 = VectorParameter() {
   group=TextureSampler
   parameterName=TextureFlow2_G2C_Refine
   channelNames {a=Blend, b=B_Mul, g=A_Mul, r=A_Power}
   defaultValue=<rgba 1 1 1 0.5>
  }
  n50 = Add(
   A=Multiply(
    A=VectorParameter() {group=TextureSampler, parameterName=TextureSampler_G2C_ColorB, defaultValue=<rgba 0 1.6667 5 1>}.RGB
    B=MaterialFunctionCall(Texture=@n45.R, RGB=@n38, A=@n38) {materialFunction=&MF_FX_Refine}.Result
   )
   B=Multiply(
    A=VectorParameter() {group=TextureSampler, parameterName=TextureF2_G2C_ColorB, defaultValue=<rgba 0.4167 0 5 1>}.RGB
    B=MaterialFunctionCall(Texture=@n48.R, RGB=@n49, A=@n49) {materialFunction=&MF_FX_Refine}.Result
   )
  )
  n51 = MaterialFunctionCall(Float2=TextureCoordinate()) {materialFunction=&BreakOutFloat2Components}
  n52 = Power(Base=Multiply(A=OneMinus(None=@n51.G)) {constB=25})
  n53 = Multiply(A=Saturate(None=Multiply(A=@n36, B=@n37)), B=Multiply(A=Multiply(A=@n50, B=@n51.G), B=Saturate(None=@n52)))
  n54 = LinearInterpolate(
   A=VectorParameter() {parameterName=Color, defaultValue=<rgba 5 4.75 0 1>}.RGB
   B=Add(A=Multiply(A=@n23, B=@n24.RGB), B=Add(A=@n35, B=@n53))
   Alpha=@n2
  )
  Multiply(
   A=ParticleColor().RGB
   B=MaterialFunctionCall(Emissive=Divide(A=@n54) {constB=1.977}) {materialFunction=&MF_FX_EmissiveAdjustments}.Result
  )
  ScalarParameter() {defaultValue=0.05, parameterName=R}
  ScalarParameter() {defaultValue=1, parameterName=S}
  TextureSample() {samplerType=SAMPLERTYPE_Masks, texture=/BRCosmetics/Weapons/FORT_Melee/Pickaxe_InstantGravelNoble/Textures/T_Pickaxe_InstantGravelNoble_S}
 ]
}
