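# Test translucent, unlit, two-sided center glow material for the InstantGravelNoble pickaxe, usable on Niagara mesh particles.
# A sine gradient along V blends two colors, with a slow time-based brightness pulse feeding emissive adjustments.
# The same gradient varies SplineThicken width, whose normal drives an inverted Fresnel fade.
# That fade combines with U and V edge masks and a panning energy texture sampled on V only to form the alpha mask.
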Material /BRCosmetics/Weapons/FORT_Melee/Pickaxe_InstantGravelNoble/Meshes/Test/M_CenterGlow
graph {
 properties {
  bUsedWithNiagaraMeshParticles=true
  blendMode=BLEND_Translucent
  pixelDepthOffsetMode=PDOM_Legacy
  shadingModel=MSM_Unlit
  twoSided=true
  usedShadingModels=Unlit
 }
 nodes [
  n1 = ScalarParameter() {defaultValue=0.5, parameterName=AWFAW}
  n2 = Multiply(A=MaterialFunctionCall(Float2=TextureCoordinate()) {materialFunction=&BreakOutFloat2Components}.G)
  n3 = OneMinus(None=Sine(None=Multiply(A=@n1, B=@n2)))
  n4 = LinearInterpolate(
   A=VectorParameter() {parameterName=C, defaultValue=<rgba 0.05 0.3983 1 1>}.RGB
   B=VectorParameter() {parameterName=c3, defaultValue=<rgba 0.26 0.2 2 1>}.RGB
   Alpha=@n3
  )
  n5 = Multiply(A=@n4)
  n6 = Multiply(A=@n4) {constB=1.3}
  n7 = Multiply(A=Time() {bOverride_Period=true, period=6666}) {constB=0.25}
  n8 = Multiply(A=Add(A=Sine(None=@n7))) {constB=0.5}
  MaterialFunctionCall(Emissive=Divide(A=LinearInterpolate(A=@n5, B=@n6, Alpha=@n8)) {constB=1}) {materialFunction=&MF_FX_EmissiveAdjustments}
  n9 = TextureObjectParameter() {
   group=MainEnergyFlow
   parameterName=MainEnergyFlow_Texture
   sortPriority=0
   texture=/BRCosmetics/Weapons/FORT_Melee/Pickaxe_Tower_Sentinel_Female/FX/T_TowerSentinel_trail
  }
  n10 = VectorParameter() {
   group=MainEnergyFlow
   parameterName=MainEnergyFlow_UVControl
   channelNames {a=VPan, b=UPan, g=VScale, r=UScale}
   defaultValue=<rgba 1 1 0.5 0>
  }
  n11 = MaterialFunctionCall(Float2=TextureCoordinate()) {materialFunction=&BreakOutFloat2Components}
  n12 = Power(Base=MaterialFunctionCall(
    Texture=@n9
    UVControl_RGB=@n10
    UVControl_A=@n10
    "[Override]UVInput"=MaterialFunctionCall(X=@n11.G, Y=@n11.G) {materialFunction=&MakeFloat2}.Result
   ) {materialFunction=&FXCommon_TextureSample}.R) {constExponent=0.5}
  n13 = VectorParameter() {group=EdgeMS, parameterName=EdgeMS_EdgeMask_InvertControl, channelNames {a="", b=Power, g=Multiply, r=Offset}, defaultValue=<rgba 0 15 4 1>}
  n14 = VectorParameter() {group=EdgeMS, parameterName=EdgeMS_EdgeMask_Control, channelNames {a="", b=Power, g=Multiply, r=Offset}, defaultValue=<rgba 0 15 4 1>}
  n15 = LinearInterpolate(A=ScalarParameter() {defaultValue=2, parameterName=ate}, B=ScalarParameter() {defaultValue=5, parameterName=ate123512}, Alpha=@n3)
  n16 = MaterialFunctionCall(WidthBase=@n15, WidthTip=@n15) {materialFunction=&SplineThicken}
  n17 = Multiply(
   A=MaterialFunctionCall!(UorV=@n11.R, Invert_EdgeControl=@n13, NoInvert_EdgeControl=@n14) {materialFunction=/Game/Effects/Niagara/MasterMaterials/Functions/Common/FXCommon_EdgeMask_Ex}.Mask
   B=OneMinus(None=Fresnel(Normal=@n16.Normal) {exponent=2})
  )
  n18 = ScalarParameter() {defaultValue=0.15, parameterName=TB}
  n19 = VectorParameter() {group=TB, parameterName=TB_EdgeMask_InvertControl, channelNames {a="", b=Power, g=Multiply, r=Offset}, defaultValue=<rgba 0 5 2 1>}
  n20 = VectorParameter() {group=TB, parameterName=TB_EdgeMask_Control, channelNames {a="", b=Power, g=Multiply, r=Offset}, defaultValue=<rgba 0 5 2 1>}
  n21 = MaterialFunctionCall(UorV=@n11.G, Invert_EdgeControl=@n19, NoInvert_EdgeControl=@n20) {materialFunction=&FXCommon_EdgeMask_Ex}
  Multiply(
   A=Multiply(A=Multiply(A=@n12, B=@n17), B=LinearInterpolate(A=@n18, Alpha=@n21.Mask))
   B=ScalarParameter() {defaultValue=1, parameterName=Scaleo}
  )
 ]
}
