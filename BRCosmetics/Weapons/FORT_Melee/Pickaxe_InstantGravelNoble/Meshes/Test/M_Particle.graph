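# Translucent, unlit, two-sided test particle material for Niagara sprites, stored with the InstantGravelNoble pickaxe cosmetic.
# An inverted V texture coordinate forms a vertical gradient without any texture, tinted by particle color RGB and scaled down by a constant divide.
# The result feeds MF_FX_EmissiveAdjustments, whose output has no visible link to a material output.
# Particle alpha is unused.
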Material /BRCosmetics/Weapons/FORT_Melee/Pickaxe_InstantGravelNoble/Meshes/Test/M_Particle
graph {
 properties {
  bUsedWithNiagaraSprites=true
  blendMode=BLEND_Translucent
  pixelDepthOffsetMode=PDOM_Legacy
  shadingModel=MSM_Unlit
  twoSided=true
  usedShadingModels=Unlit
 }
 nodes [
  n1 = OneMinus(None=MaterialFunctionCall(Float2=TextureCoordinate()) {materialFunction=&BreakOutFloat2Components}.G)
  n2 = ParticleColor()
  MaterialFunctionCall(Emissive=Divide(A=Multiply(A=@n1, B=@n2.RGB)) {constB=1.977}) {materialFunction=&MF_FX_EmissiveAdjustments}
 ]
}
