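# Unlit test material for a ring mesh on a pickaxe cosmetic, enabled for Niagara mesh particles.
# Mirrored power gradients along local-position Y create tinted bands, added to a GGX highlight, Fresnel tint and base emissive color.
# The sum passes through an emissive-adjustment function and is multiplied by particle color.
# A separate branch multiplies a remapped Y gradient by inverse Fresnel.
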Material /BRCosmetics/Weapons/FORT_Melee/Pickaxe_InstantGravelNoble/Meshes/Test/M_Ring1
graph {
 properties {bUsedWithNiagaraMeshParticles=true, pixelDepthOffsetMode=PDOM_Legacy, shadingModel=MSM_Unlit, usedShadingModels=Unlit}
 nodes [
  n1 = MaterialFunctionCall(Float3=MaterialFunctionCall() {materialFunction=&LocalPosition}.Local Position) {materialFunction=&BreakOutFloat3Components}
  n2 = Multiply(A=@n1.G) {constB=-1}
  n3 = ScalarParameter() {defaultValue=5, parameterName=YddddDist15}
  n4 = ScalarParameter() {defaultValue=2, parameterName=YddddPower25}
  n5 = Saturate(None=Power(Base=Divide(A=@n2, B=@n3), Exp=@n4))
  n6 = Saturate(None=Power(Base=Divide(A=@n1.G, B=@n3), Exp=@n4))
  n7 = ScalarParameter() {defaultValue=2, parameterName=wdawfge}
  Multiply(A=LinearInterpolate(Alpha=Add(A=@n5, B=@n6)) {constA=0.4, constB=0.85}, B=OneMinus(None=Fresnel(ExponentIn=@n7) {baseReflectFraction=0}))
  n8 = VectorParameter() {parameterName=245, defaultValue=<rgba 0.05 0.2 0.5 0.5>}
  n9 = ScalarParameter() {defaultValue=6, parameterName=YDist15}
  n10 = Divide(A=Multiply(A=@n1.G) {constB=-1}, B=@n9)
  n11 = ScalarParameter() {defaultValue=30, parameterName=YPower25}
  n12 = Divide(A=@n1.G, B=@n9)
  n13 = Add(A=Saturate(None=Power(Base=@n10, Exp=@n11)), B=Saturate(None=Power(Base=@n12, Exp=@n11)))
  n14 = ScalarParameter() {defaultValue=6, parameterName=YDist}
  n15 = ScalarParameter() {defaultValue=5.0849, parameterName=YPower2}
  n16 = Power(Base=Divide(A=Multiply(A=@n1.G) {constB=-1}, B=@n14), Exp=@n15)
  n17 = Power(Base=Divide(A=@n1.G, B=@n14), Exp=@n15)
  n18 = Multiply(A=VectorParameter() {parameterName=12341, defaultValue=<rgba 0.03 0.1515 0.3 0.5>}.RGB, B=Add(A=Saturate(None=@n16), B=Saturate(None=@n17)))
  n19 = ScalarParameter() {parameterName=R}
  n20 = ScalarParameter() {defaultValue=0.7, parameterName=sssEX}
  n21 = Add(
   A=Multiply(
    A=MaterialFunctionCall!(LightDirctionIn, CameraVectorIn, Roughness=@n19, NormalIn) {materialFunction=/Game/Effects/Niagara/MasterMaterials/Functions/Common/FXCommon_GGX}.Result
    B=VectorParameter() {parameterName=Emissive, defaultValue=<rgba 0 0.1333 1 1>}.RGB
   )
   B=Multiply(A=VectorParameter() {parameterName=FresC, defaultValue=<rgba 0.05 0 0.6 1>}.RGB, B=Fresnel(ExponentIn=@n20))
  )
  n22 = Add(
   A=Add(A=Multiply(A=@n8.RGB, B=@n13), B=Add(A=@n18, B=@n21) {constB=0})
   B=VectorParameter() {parameterName=Emissive2, defaultValue=<rgba 0.108 0.1212 0.9 1>}.RGB
  ) {constB=0}
  Multiply(
   A=ParticleColor().RGB
   B=MaterialFunctionCall(Emissive=Divide(A=@n22) {constB=1.7}) {materialFunction=&MF_FX_EmissiveAdjustments}.Result
  )
  ScalarParameter() {defaultValue=1}
  ScalarParameter() {defaultValue=1, parameterName=F}
  ScalarParameter() {defaultValue=1, parameterName=M}
  VectorParameter() {parameterName=BaseColor, defaultValue=<rgba 0.08 0.1067 0.4 1>}
 ]
}
